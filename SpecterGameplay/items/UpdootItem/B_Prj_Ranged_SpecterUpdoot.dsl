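; Homing projectile for the Specter Updoot weapon, locking onto a replicated target and hidden when it has no valid instigator.
; BeginPlay caches the instigator's weapon and starts cosmetic trail and audio.
; On stop, homing ends; cosmetic clients play impact effects, while the server applies the updoot gameplay effect to the target, then spawns a doot pickup after a delay.
; Destruction triggers both server actions as a fallback, with DoOnce gates preventing repeats.
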(event EventBeginPlay
  (bind _returnvalue (Game|GetInstigator))
  ;[1 1 1] EVENT -> Begin Play
  (|CallParentFunction)
  (Utilities|IsValid _returnvalue
    (:Is%20Valid
      (bind _asfort_player_pawn (Utilities|Casting|CastToFortPlayerPawn _returnvalue)
        (:then
          (bind _asb_specter_updoot_weapon (Utilities|Casting|CastToB_Specter_Updoot_Weapon (Variables|/Script/FortniteGame.FortPawn|GetCurrentWeapon _asfort_player_pawn)))
          (Variables|Default|SetCached%20B%20Specter%20Updoot%20Weapon _asb_specter_updoot_weapon)
          ;;[1 1 1] Spawn attached Trail
          (Networking|CanExecuteCosmeticEvents
            (:True
              (Variables|Default|SetIsLocallyVisible true)
              (Components|Activation|Activate (Variables|Default|GetAudioComponent) false)
              (bind _returnvalue_2 (Effects|SpawnFXSystemAttached (Variables|Trail|GetTrail%20Asset) (Variables|Collision|GetCapsuleComponent)))
              (bind _asniagara_particle_system_component (Utilities|Casting|CastToNiagaraComponent _returnvalue_2))
              (Variables|Default|SetTrailPS_Niagara _asniagara_particle_system_component))))
        (:CastFailed
          _)))
    (:Is%20Not%20Valid
      _))
  ; If we don't have an instigator, we're likely not a ghost, and shouldn't be seeing this.
  (Variables|Default|SetIsLocallyVisible false)
  (Development|SetHiddeninGame (Variables|Projectile|GetMesh) true true))

(event EventOnStop
  (params (Hit :type Hit%20Result%20Structure%20%28by%20ref%29))
  (bind (_bblockinghit _binitialoverlap _time _distance _location _impactpoint _normal _impactnormal) (Collision|BreakHitResult Hit))
  (bind _trailps_niagara (Variables|Default|GetTrailPS_Niagara))
  ;[1 1 1] EVENT -> Stop: Impact and Explosion VFX Stuff
  (Variables|Effects|SetStopLocation _impactpoint)
  (Variables|Effects|SetStopRotZ (Math|Rotator|MakeRotfromZ _impactnormal))
  (Utilities|FlowControl|DoOnce false
    (:Completed
      (Projectile|SetHomingStyle 0 0 "None" false :!class /Script/FortniteGame.FortProjectileMovementInterface)
      (Networking|CanExecuteCosmeticEvents
        (:True
          (Development|SetHiddeninGame (Variables|Projectile|GetMesh) true true)
          (if (Variables|Default|GetIsLocallyVisible)
            (Utilities|IsValid _trailps_niagara
              (:Is%20Valid
                (Components|Activation|Deactivate _trailps_niagara))
              (:Is%20Not%20Valid
                _))
            ;;[0.2043 0 1 1 30] Niagara Hit VFX
            (Effects|Components|ParticleSystem|SpawnFXAtLocation (Variables|Effects|Explosion|GetExplosion%20Asset) (Variables|Effects|GetStopLocation) (Variables|Effects|GetStopRotZ))
            (Audio|PlaySoundatLocation (Variables|Sounds|GetProjectile_Explosion_Sound) (Variables|Effects|GetStopLocation))
            ;;[0.6661 0.1729 1] Shake camera
            (GameplayCue|ExecuteGameplayCue%28Burst%29Local self (Variables|Effects|GetFeedbackCue) (Ability|GameplayCue|MakeGameplayCueParameters 0.0)) ; Cam shake + controller vibration
            (bind _audiocomponent (Variables|Default|GetAudioComponent)
              (:then
                (Audio|Components|Audio|FadeOut _audiocomponent 0.15 0.0 "Linear" :!class /Script/Engine.AudioComponent))
              (:else))))
        (:False
          (CallFunction|OnServerHit))))))

(event EventDestroyed
  (bind _trailps_niagara (Variables|Default|GetTrailPS_Niagara))
  ;[1 1 1] EVENT -> Destroyed: Deactivating Trail Particle systems on component destroyed & spawn doot pickup if we haven't.
  (Networking|CanExecuteCosmeticEvents
    (:True
      (Utilities|IsValid _trailps_niagara
        (:Is%20Valid
          (Components|Activation|Deactivate _trailps_niagara)
          (bind _audiocomponent (Variables|Default|GetAudioComponent)
            (:then
              (Audio|Components|Audio|FadeOut _audiocomponent 0.15 0.0 "Linear" :!class /Script/Engine.AudioComponent))
            (:else)))))
    (:False
      (CallFunction|SpawnUpdootHitGCNOnce)
      (CallFunction|SpawnDootPickupOnce))))

(event Custom|OnServerHit
  ; On Server Hit
  (CallFunction|SpawnUpdootHitGCNOnce)
  ;; Delay and spawn pickup
  (Utilities|FlowControl|Delay (ScalableFloat|GetValueatLevel (Variables|Default|GetSpawnPickupDelaySeconds)))
  (CallFunction|SpawnDootPickupOnce))

(event Custom|SpawnDootPickupOnce
  ; Spawn Doot Pickup Once
  (Networking|SwitchHasAuthority
    (:Authority
      (Utilities|FlowControl|DoOnce false
        (:Completed
          (Class|BSpecterUpdootWeapon|SpawnDootPickup (Variables|Default|GetCached%20B%20Specter%20Updoot%20Weapon) (Variables|Default|GetHoming%20Target%20Repped)))))))

(event Custom|SpawnUpdootHitGCNOnce
  ; Spawn Updoot GCN on target player if we haven't.
  (Networking|SwitchHasAuthority
    (:Authority
      (Utilities|FlowControl|DoOnce false
        (:Completed
          (bind _asfort_player_pawn (Utilities|Casting|CastToFortPlayerPawn (Variables|Default|GetHoming%20Target%20Repped)))
          (Class|BSpecterUpdootWeapon|ApplyUpdootGameplayEffect (Variables|Default|GetCached%20B%20Specter%20Updoot%20Weapon) _asfort_player_pawn))))))

(fn OnRep_HomingTargetRepped
  (CallFunction|SetHomingVars))

(fn SetHomingVars
  (Projectile|SetHomingData (Variables|Movement|GetProjectileMovementComponent) (Utilities|Struct|MakeProjectileHomingData "LockOn" 1.0 1e+03 1.0 0.0 -1 (Variables|Default|GetHoming%20Target%20Repped) 0 "0.000000,0.000000,0.000000" (Variables|Default|GetTargetHomingOffset) "Auto" false true) :!class /Script/FortniteGame.FortProjectileMovementComponent))

(fn ConstructionScript)
